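# Sound cue for the opening sound of the ZebraScramble Bacon glider cosmetic in Battle Royale.
# A mixer layers the open sound with one of five equally weighted laugh variants, picked at random behind a delay node left at its default settings.
# Glider open/close attenuation and concurrency, a cosmetic glider-open property sheet and a glider sound class govern playback.
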SoundCue /BRCosmetics/Sounds/Gliders/ZebraScramble_Bacon/GliderOpen_ZebraScramble_Bacon_Cue1
root=Mixer_0
nodes [
 {id=Delay_0, children [Random_2]}
 {id=Mixer_0, children [Random_0,Delay_0], properties {inputVolume [1,1]}}
 {id=Random_0, children [WavePlayer_0], properties {weights [1]}}
 {id=Random_2, children [WavePlayer_17, WavePlayer_18, WavePlayer_19, WavePlayer_20, WavePlayer_21], properties {weights [1,1,1,1,1]}}
 {id=WavePlayer_0, properties {soundWaveAssetPtr=/Game/Sounds/Fort_Gadgets/Athena/ZebraScramble_Bacon/IP_Glider_ZebraScrambleBacon_Open}}
 {id=WavePlayer_17, properties {soundWaveAssetPtr=&/IP_Glider_ZebraScrambleBacon_Laugh_Open_01}}
 {id=WavePlayer_18, properties {soundWaveAssetPtr=&/IP_Glider_ZebraScrambleBacon_Laugh_Open_02}}
 {id=WavePlayer_19, properties {soundWaveAssetPtr=&/IP_Glider_ZebraScrambleBacon_Laugh_Open_03}}
 {id=WavePlayer_20, properties {soundWaveAssetPtr=&/IP_Glider_ZebraScrambleBacon_Laugh_Open_04}}
 {id=WavePlayer_21, properties {soundWaveAssetPtr=&/IP_Glider_ZebraScrambleBacon_Laugh_Open_05}}
]
properties {
 attenuationSettings=/Game/Sounds/Attenuation/Fort_Player_Glider_OpenClose_Athena_Att
 audioPropertiesSheet=/BRCosmetics/Sounds/Settings/PropertySheets/Glider/APS_Cosmetic_Glider_Open
 maxDistance=10000
 soundClassObject=&Fort_Gliders_Land
 volumeMultiplier=0.8
 concurrencySet [&Athena_GliderOpenClose_Conc]
}
