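# Sound cue for the thrust-start sound of the ZebraScramble Bacon glider, pairing a thrust burst with a laugh.
# A mixer layers a randomly chosen thrust-start wave, pitch-varied through a modulator, with a random laugh from a larger pool that follows after a fixed one-second delay.
# The cue uses glider-thrust attenuation and concurrency settings and is routed to the glider landing sound class.
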SoundCue /BRCosmetics/Sounds/Gliders/ZebraScramble_Bacon/GliderThrustStart_ZebraScramble_Bacon_Cue1
root=Mixer_0
nodes [
 {id=Delay_0, children [Random_1], properties {delayMax=1, delayMin=1}}
 {id=Mixer_0, children [Modulator_0,Delay_0], properties {inputVolume [1,1]}}
 {id=Modulator_0, children [Random_0], properties {pitchMax=1.2, pitchMin=0.8}}
 {id=Random_0, children [WavePlayer_0,WavePlayer_1,WavePlayer_2], properties {weights [1,1,1]}}
 {
  id=Random_1
  children [
   WavePlayer_20
   WavePlayer_3
   WavePlayer_4
   WavePlayer_5
   WavePlayer_6
   WavePlayer_8
   WavePlayer_9
   WavePlayer_10
   WavePlayer_11
   WavePlayer_12
   WavePlayer_13
   WavePlayer_14
   WavePlayer_15
   WavePlayer_16
  ]
  properties {weights [1,1,1,1,1,1,1,1,1,1,1,1,1,1]}
 }
 {
  id=WavePlayer_0
  properties {soundWaveAssetPtr=/Game/Sounds/Fort_Gadgets/Athena/ZebraScramble_Bacon/IP_Glider_ZebraScrambleBacon_ThurstStart_03}
 }
 {
  id=WavePlayer_1
  properties {soundWaveAssetPtr=/Game/Sounds/Fort_Gadgets/Athena/ZebraScramble_Bacon/IP_Glider_ZebraScrambleBacon_ThurstStart_02}
 }
 {id=WavePlayer_10, properties {soundWaveAssetPtr=&/IP_Glider_ZebraScrambleBacon_Laugh_10}}
 {id=WavePlayer_11, properties {soundWaveAssetPtr=&/IP_Glider_ZebraScrambleBacon_Laugh_11}}
 {id=WavePlayer_12, properties {soundWaveAssetPtr=&/IP_Glider_ZebraScrambleBacon_Laugh_12}}
 {id=WavePlayer_13, properties {soundWaveAssetPtr=&/IP_Glider_ZebraScrambleBacon_Laugh_13}}
 {id=WavePlayer_14, properties {soundWaveAssetPtr=&/IP_Glider_ZebraScrambleBacon_Laugh_14}}
 {id=WavePlayer_15, properties {soundWaveAssetPtr=&/IP_Glider_ZebraScrambleBacon_Laugh_15}}
 {id=WavePlayer_16, properties {soundWaveAssetPtr=&/IP_Glider_ZebraScrambleBacon_Laugh_16}}
 {
  id=WavePlayer_2
  properties {soundWaveAssetPtr=/Game/Sounds/Fort_Gadgets/Athena/ZebraScramble_Bacon/IP_Glider_ZebraScrambleBacon_ThurstStart_01}
 }
 {id=WavePlayer_20, properties {soundWaveAssetPtr=&/IP_Glider_ZebraScrambleBacon_Laugh_02}}
 {id=WavePlayer_3, properties {soundWaveAssetPtr=&/IP_Glider_ZebraScrambleBacon_Laugh_03}}
 {id=WavePlayer_4, properties {soundWaveAssetPtr=&/IP_Glider_ZebraScrambleBacon_Laugh_04}}
 {id=WavePlayer_5, properties {soundWaveAssetPtr=&/IP_Glider_ZebraScrambleBacon_Laugh_05}}
 {id=WavePlayer_6, properties {soundWaveAssetPtr=&/IP_Glider_ZebraScrambleBacon_Laugh_06}}
 {id=WavePlayer_8, properties {soundWaveAssetPtr=&/IP_Glider_ZebraScrambleBacon_Laugh_08}}
 {id=WavePlayer_9, properties {soundWaveAssetPtr=&/IP_Glider_ZebraScrambleBacon_Laugh_09}}
]
properties {
 attenuationSettings=&Fort_Player_Glider_Thrust_Athena_Att
 maxDistance=12500
 soundClassObject=&Fort_Gliders_Land
 volumeMultiplier=1
 concurrencySet [&Athena_Glider_Thrust_Conc]
}
